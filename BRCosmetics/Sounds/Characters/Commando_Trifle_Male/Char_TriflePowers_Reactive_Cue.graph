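# Sound cue for the reactive foley sound of the Commando Trifle Male character cosmetic, in the player foley sound class.
# A random node picks one of six equally weighted Reactive wave variations and feeds it through a modulator for slight variation.
# Playback uses basic foley attenuation and a boosted volume.
# A concurrency override allows one instance at a time, stopping the oldest when a new one starts.
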SoundCue /BRCosmetics/Sounds/Characters/Commando_Trifle_Male/Char_TriflePowers_Reactive_Cue
root=Modulator_0
nodes [
 {id=Modulator_0, children [Random_0], properties {volumeMin=1}}
 {id=Random_0, children [WavePlayer_0,WavePlayer_1,WavePlayer_2,WavePlayer_3,WavePlayer_4,WavePlayer_5], properties {weights [1,1,1,1,1,1]}}
 {id=WavePlayer_0, properties {soundWaveAssetPtr=&/Char_Commando_TriflePowers_Reactive_01}}
 {id=WavePlayer_1, properties {soundWaveAssetPtr=&/Char_Commando_TriflePowers_Reactive_02}}
 {id=WavePlayer_2, properties {soundWaveAssetPtr=&/Char_Commando_TriflePowers_Reactive_03}}
 {id=WavePlayer_3, properties {soundWaveAssetPtr=&/Char_Commando_TriflePowers_Reactive_04}}
 {id=WavePlayer_4, properties {soundWaveAssetPtr=&/Char_Commando_TriflePowers_Reactive_05}}
 {id=WavePlayer_5, properties {soundWaveAssetPtr=&/Char_Commando_TriflePowers_Reactive_06}}
]
properties {
 attenuationSettings=&Foley_Attenuation_Basic_Att
 bOverrideConcurrency=true
 maxDistance=1536
 soundClassObject=&Fort_Player_Foley
 volumeMultiplier=2
 concurrencyOverrides {maxCount=1, resolutionRule=StopOldest}
}
